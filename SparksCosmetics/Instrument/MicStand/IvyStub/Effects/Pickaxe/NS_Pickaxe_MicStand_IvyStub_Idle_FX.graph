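# Niagara idle effect for the IvyStub mic stand cosmetic's pickaxe form, adding green mist, spiraling ribbons and glowing motes.
# Mist sprites and ribbons attach to skeletal mesh sockets, with time-of-day color scaling and one world-style mist emitter disabled.
# Burst-spawned ribbon points are repositioned each update along offset paths with sine/cosine spirals, some shaped by User.IsPickaxe.
# Motes read the spiral ribbon through a FollowRibbon module and ease toward it.
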NiagaraSystem /SparksCosmetics/Instrument/MicStand/IvyStub/Effects/Pickaxe/NS_Pickaxe_MicStand_IvyStub_Idle_FX
dependencies {
 usedDataInterfaces [
  &NiagaraDataInterfaceArrayFloat
  &NiagaraDataInterfaceArrayInt32
  &NiagaraDataInterfaceCurve
  &NiagaraDataInterfaceColorCurve
  &NiagaraDataInterfaceVectorCurve
  &NiagaraDataInterfaceParticleRead
  &NiagaraDataInterfaceVector2DCurve
  &NiagaraDataInterfaceSpriteRendererInfo
  &NiagaraDataInterfaceSkeletalMesh
 ]
 usedDynamicInputs [&MakeFloatFromVector]
 usedModules [
  &self:SystemSpawnScript.SystemScriptSource.SystemScriptGraph.NiagaraNodeAssignment_0.SetVariables_#1
  &self:SystemSpawnScript.SystemScriptSource.SystemScriptGraph.NiagaraNodeAssignment_1.SetVariables_#2
  &self:SystemSpawnScript.SystemScriptSource.SystemScriptGraph.NiagaraNodeAssignment_2.SetVariables_#3
  &SystemState
  &self:SystemSpawnScript.SystemScriptSource.SystemScriptGraph.NiagaraNodeAssignment_3.SetVariables_#4
  &Timeline
 ]
}
emitters {
 Mist_Sprites001 {
  bEnabled=false
  emitterUpdateScript [
   &EmitterState(Life Cycle Mode=::NewEnumerator1, Loop Duration=0.1, UseLoopDelay=false)
   &SpawnPerUnit(
    Module.Use Max Movement Threshold=true
    Module.Use Movement Tolerance=false
    Spawn Probability=0.7
    Spawn Spacing=200
    Use Spawn Probability=true
   )
   &SpawnRate(Spawn Probability=0.8, SpawnRate=5, Use Spawn Probability=true)
  ]
  particleSpawnScript [
   &PartitionParticles(Number Of Partitions=4)
   &InitializeParticle(
    Color=<rgba 0.0547 0.75 0.4491 0.35>
    Lifetime Max=2.5
    Lifetime Min=1.75
    Lifetime Mode=::NewEnumerator1
    Mass Mode=::NewEnumerator3
    Sprite Rotation Mode=::NewEnumerator2
    Sprite Size Mode=::NewEnumerator1
    Uniform Sprite Size Max=70
    Uniform Sprite Size Min=50
    UsePositionOffset=false
    Write Material Random=true
   )
   &ShapeLocation(
    Apply To Particle Position=::NewEnumerator1
    MaskSpawnGroup=false
    Non Uniform Scale=<1 1 0.25>
    Override Seed=false
    Shape Origin=&VectorFromFloat(Value=0)
    Sphere Radius=&SelectFloatFromArray(Direct Array Index=$Particles.PartitionParticles.Partition, Float Selection Array=DataInterface() {propertyValues {FloatData [6,9,12,15]}})
    UseOwnerScale=true
   )
   &SkeletalMeshLocation(
    MaskSpawnGroup=false
    OverrideRandomSeed=false
    Position Sampling=::NewEnumerator1
    Sampled Position Offset:NiagaraPosition=$Particles.Initial.ShapeLocation.ShapePosition
    Skeletal Mesh=$System.Skeletal Mesh
    Use Sampled Mesh Attribute as Ribbon ID=false
   ) {enabled=false, module=SkeletalMeshLocation001}
   &SkeletalMeshLocation(
    Bone / Socket Index=&SelectIntFromArray(Direct Array Index=$Particles.PartitionParticles.Partition, Int Selection Array=DataInterface() {propertyValues {IntData [2,2,2,2]}})
    MaskSpawnGroup=false
    Mesh Sampling Type=::NewEnumerator5
    Position Sampling=::NewEnumerator1
    Sampled Position Offset:NiagaraPosition=$Particles.Initial.ShapeLocation.ShapePosition
    Skeletal Mesh=$System.Skeletal Mesh
    Socket Sampling Mode=::NewEnumerator5
    Use Sampled Mesh Attribute as Ribbon ID=false
   ) {module=SkeletalMeshLocation002}
   &AddVelocity(
    Velocity=&RandomRangeVector(Minimum=<0 0 20>, Maximum=<1 1 150>)
    Velocity Speed Scale=&SelectFloatFromArray(Direct Array Index=$Particles.PartitionParticles.Partition, Float Selection Array=DataInterface() {propertyValues {FloatData [2,1,0.5,1.2]}})
   ) {enabled=false}
   &CameraOffset(Camera Offset Amount=15)
  ]
  particleUpdateScript [
   &ParticleState()
   &SkeletalMeshLocation(
    Bone / Socket Index=&SelectIntFromArray(Direct Array Index=$Particles.PartitionParticles.Partition, Int Selection Array=DataInterface() {propertyValues {IntData [2,2,2,2]}})
    MaskSpawnGroup=false
    Mesh Sampling Type=::NewEnumerator5
    Sampled Position Offset=&AddVectorToPosition(Position=$Particles.Initial.ShapeLocation.ShapePosition, Vector=<10 0 0>)
    Skeletal Mesh=$System.Skeletal Mesh
    Socket Sampling Mode=::NewEnumerator5
    Use Sampled Mesh Attribute as Ribbon ID=false
   )
   /Niagara/Modules/Update/SubUV/SubUVAnimation(Number Of Frames=63)
   &SpriteRotationRate(Rotation Rate=&RandomRangeFloat(Minimum=15, Maximum=50))
   &ScaleSpriteSize(
    Scale Sprite Size Mode=::NewEnumerator0
    Uniform Scale Factor=&FloatFromCurve(
     FloatCurve=DataInterface() {propertyValues {Curve {keys [
         {arriveTangent=1.0536, interpMode=RCIM_Cubic, leaveTangent=1.0536, tangentMode=RCTM_User, time=0, value=0.25}
         {arriveTangent=0.471, interpMode=RCIM_Cubic, leaveTangent=0.471, tangentMode=RCTM_User, time=0.7, value=0.825}
         {arriveTangent=4.5585, interpMode=RCIM_Cubic, leaveTangent=4.5585, tangentMode=RCTM_User, time=1, value=1.275}
        ]}}}
     Scale Curve=0.75
    )
   )
   &ScaleColor(
    Linear Color Curve=DataInterface() {propertyValues {
      AlphaCurve {keys [<at 0 0>,<at 0.0645 1>,<at 0.2106 0.35>,<at 1 0>]}
      BlueCurve {keys [<at 0 1>,<at 1 0.5312>]}
      GreenCurve {keys [<at 0 1>,<at 1 0.2802>]}
      RedCurve {keys [<at 0 1>,<at 1 0>]}
     }}
    Scale Mode=::NewEnumerator2
   )
   &Drag(Drag=4, Use Linear Drag=true, Use Rotational Drag=true)
   &SolveForcesAndVelocity(Clamp Velocity=true, Limit Acceleration=true)
   &ScaleColor(
    Scale Alpha=&Multiply_Float(
     A=&FloatFromCurve(FloatCurve=DataInterface() {propertyValues {Curve {keys [
          {interpMode=RCIM_Cubic, leaveTangentWeight=0.2339, tangentMode=RCTM_User, time=0, value=0}
          {arriveTangentWeight=0.0667, interpMode=RCIM_Cubic, leaveTangentWeight=0.2668, tangentMode=RCTM_User, time=0.2, value=1}
          {arriveTangentWeight=0.2701, interpMode=RCIM_Cubic, tangentMode=RCTM_User, time=1, value=0}
         ]}}})
     B=&FNTimeOfDayScalar(Night Scale=0.8, Morning Scale=0.7, Day Scale=0.45, Evening Scale=0.6)
    )
    Scale RGB=&FN_Select_Vector_from_Time_Of_Day(Morning=<0.9 0.9 0.9>, Day=<1 1 1>, Evening=<1 1 1>, Night=<0.5 0.5 1>)
    ScaleA=true
    ScaleRGB=true
   ) {module=ScaleColor002}
   &InterpolateOverTime(
    Rate Of Change:NiagaraFloat=&FloatFromCurve(FloatCurve=DataInterface() {propertyValues {Curve {keys [
         {arriveTangent=-143.9088, interpMode=RCIM_Cubic, leaveTangent=-143.9088, tangentMode=RCTM_User, time=0.1, value=50}
         {arriveTangent=-168.5692, interpMode=RCIM_Cubic, leaveTangent=-352.3025, tangentMode=RCTM_Break, time=0.35, value=17}
         {arriveTangent=-9.0214, interpMode=RCIM_Cubic, leaveTangent=-9.0214, tangentMode=RCTM_User, time=0.45, value=1}
         <at 0.6 0 RCIM_Cubic>
        ]}}})
    Target Position Value:NiagaraPosition=$Output.SolveForcesAndVelocity.Position
    Variable Type=::NewEnumerator6
   )
   /SetVariables(Particles.Position=&AddVectorToPosition(
     Position=$StackContext.InterpolateOverTime.InterpolateOverTime_Position.Moving Average
     Vector=&Multiply_Vector(
      A=&MakeVector(
       X=&FloatFromCurve(
        FloatCurve=DataInterface() {propertyValues {Curve {
           defaultValue=1
           keys [
            {arriveTangent=3.5337, interpMode=RCIM_Cubic, leaveTangent=3.5337, tangentMode=RCTM_User, time=0, value=0}
            {interpMode=RCIM_Cubic, tangentMode=RCTM_User, time=1, value=1}
           ]
          }}}
        Scale Curve=&RandomRangeFloat(Minimum=-2, Maximum=2)
       )
       Y=&FloatFromCurve(
        FloatCurve=DataInterface() {propertyValues {Curve {
           defaultValue=1
           keys [
            {arriveTangent=3.5337, interpMode=RCIM_Cubic, leaveTangent=3.5337, tangentMode=RCTM_User, time=0, value=0}
            {interpMode=RCIM_Cubic, tangentMode=RCTM_User, time=1, value=1}
           ]
          }}}
        Scale Curve=&RandomRangeFloat(Minimum=-2, Maximum=2)
       )
       Z=&FloatFromCurve(
        FloatCurve=DataInterface() {propertyValues {Curve {
           defaultValue=1
           keys [
            {arriveTangent=3.5337, interpMode=RCIM_Cubic, leaveTangent=3.5337, tangentMode=RCTM_User, time=0, value=0}
            {interpMode=RCIM_Cubic, tangentMode=RCTM_User, time=1, value=1}
           ]
          }}}
        Scale Curve=&RandomRangeFloat(Minimum=-2, Maximum=2)
       )
      )
      B=&VectorFromFloat(Value=&RandomRangeFloat(Minimum=1, Maximum=2.25))
     )
    )) {isSetParameters=true, module=SetVariables_#5}
  ]
  rendererClasses [&NiagaraSpriteRendererProperties]
  renderers [{rendererClass=&NiagaraSpriteRendererProperties}]
 }
 Mist_Sprites_local {
  emitterUpdateScript [
   &EmitterState(Life Cycle Mode=::NewEnumerator1, Loop Duration=0.1, UseLoopDelay=false)
   &SpawnPerUnit(
    Module.Use Max Movement Threshold=true
    Module.Use Movement Tolerance=false
    Spawn Probability=0.7
    Spawn Spacing=200
    Use Spawn Probability=true
   ) {enabled=false}
   &SpawnRate(Spawn Probability=0.8, SpawnRate=5, Use Spawn Probability=true)
  ]
  particleSpawnScript [
   &PartitionParticles(Number Of Partitions=4)
   &InitializeParticle(
    Color=<rgba 0.0547 0.75 0.4491 0.35>
    Lifetime Max=2.5
    Lifetime Min=2
    Lifetime Mode=::NewEnumerator1
    Mass Mode=::NewEnumerator3
    Sprite Rotation Mode=::NewEnumerator2
    Sprite Size Mode=::NewEnumerator1
    Uniform Sprite Size Max=70
    Uniform Sprite Size Min=50
    UsePositionOffset=false
    Write Material Random=true
   )
   &ShapeLocation(
    Apply To Particle Position=::NewEnumerator1
    MaskSpawnGroup=false
    Non Uniform Scale=<1 1 0.25>
    Override Seed=false
    Shape Origin=&VectorFromFloat(Value=0)
    Sphere Radius=&SelectFloatFromArray(Direct Array Index=$Particles.PartitionParticles.Partition, Float Selection Array=DataInterface() {propertyValues {FloatData [6,9,12,15]}})
    UseOwnerScale=true
   )
   &SkeletalMeshLocation(
    Bone / Socket Index=&SelectIntFromArray(Direct Array Index=$Particles.PartitionParticles.Partition, Int Selection Array=DataInterface() {propertyValues {IntData [2,2,2,2]}})
    MaskSpawnGroup=false
    Mesh Sampling Type=::NewEnumerator5
    Position Sampling=::NewEnumerator1
    Sampled Position Offset:NiagaraPosition=$Particles.Initial.ShapeLocation.ShapePosition
    Skeletal Mesh=$System.Skeletal Mesh
    Socket Sampling Mode=::NewEnumerator5
    Use Sampled Mesh Attribute as Ribbon ID=false
   ) {module=SkeletalMeshLocation002}
   &AddVelocity(
    Velocity=&RandomRangeVector(Minimum=<0 0 20>, Maximum=<1 1 45>)
    Velocity Speed Scale=&SelectFloatFromArray(Direct Array Index=$Particles.PartitionParticles.Partition, Float Selection Array=DataInterface() {propertyValues {FloatData [2,1,0.5,1.2]}})
   )
   &CameraOffset(Camera Offset Amount=15)
  ]
  particleUpdateScript [
   &ParticleState()
   &SkeletalMeshLocation(
    Bone / Socket Index=&SelectIntFromArray(Direct Array Index=$Particles.PartitionParticles.Partition, Int Selection Array=DataInterface() {propertyValues {IntData [2,2,2,2]}})
    "Inherit Velocity (Scale)"=-1
    MaskSpawnGroup=false
    Mesh Sampling Type=::NewEnumerator5
    Position Sampling=::NewEnumerator1
    Sampled Position Offset=&AddVectorToPosition(Position=$Particles.Initial.ShapeLocation.ShapePosition, Vector=<10 0 0>)
    Skeletal Mesh=$System.Skeletal Mesh
    Socket Sampling Mode=::NewEnumerator5
    Use Sampled Mesh Attribute as Ribbon ID=false
    Velocity Sampling=::NewEnumerator0
    VelocitySpeedLimit=true
   )
   &SubUVAnimation(Number Of Frames=63)
   &SpriteRotationRate(Rotation Rate=&RandomRangeFloat(Minimum=15, Maximum=50))
   &ScaleSpriteSize(
    Scale Sprite Size Mode=::NewEnumerator0
    Uniform Scale Factor=&FloatFromCurve(
     FloatCurve=DataInterface() {propertyValues {Curve {keys [
         {arriveTangent=1.0536, interpMode=RCIM_Cubic, leaveTangent=1.0536, tangentMode=RCTM_User, time=0, value=0.25}
         {arriveTangent=0.471, interpMode=RCIM_Cubic, leaveTangent=0.471, tangentMode=RCTM_User, time=0.7, value=0.825}
         {arriveTangent=4.5585, interpMode=RCIM_Cubic, leaveTangent=4.5585, tangentMode=RCTM_User, time=1, value=1.275}
        ]}}}
     Scale Curve=0.75
    )
   )
   &ScaleColor(
    Linear Color Curve=DataInterface() {propertyValues {
      AlphaCurve {keys [<at 0 0>,<at 0.0645 1>,<at 0.2106 0.35>,<at 1 0>]}
      BlueCurve {keys [<at 0 1>,<at 1 0.5312>]}
      GreenCurve {keys [<at 0 1>,<at 1 0.2802>]}
      RedCurve {keys [<at 0 1>,<at 1 0>]}
     }}
    Scale Mode=::NewEnumerator2
   )
   &Drag(Drag=4, Use Linear Drag=true, Use Rotational Drag=true)
   &InheritVelocity(Speed Limit=true, UseSpeedThreshold=false) {enabled=false}
   &SolveForcesAndVelocity(Clamp Velocity=true, Limit Acceleration=true)
   &ScaleColor(
    Scale Alpha=&Multiply_Float(
     A=&FloatFromCurve(FloatCurve=DataInterface() {propertyValues {Curve {keys [
          {interpMode=RCIM_Cubic, leaveTangentWeight=0.2339, tangentMode=RCTM_User, time=0, value=0}
          {arriveTangentWeight=0.0667, interpMode=RCIM_Cubic, leaveTangentWeight=0.2668, tangentMode=RCTM_User, time=0.2, value=1}
          {arriveTangentWeight=0.2701, interpMode=RCIM_Cubic, tangentMode=RCTM_User, time=1, value=0}
         ]}}})
     B=&FNTimeOfDayScalar(Night Scale=0.85, Morning Scale=0.7, Day Scale=0.45, Evening Scale=0.6)
    )
    Scale RGB=&FN_Select_Vector_from_Time_Of_Day(Morning=<0.9 0.9 0.9>, Day=<1 1 1>, Evening=<1 1 1>, Night=<0.5 0.5 1>)
    ScaleA=true
    ScaleRGB=true
   ) {module=ScaleColor002}
   &InterpolateOverTime(
    Rate Of Change:NiagaraFloat=&FloatFromCurve(FloatCurve=DataInterface() {propertyValues {Curve {keys [
         {arriveTangent=-143.9088, interpMode=RCIM_Cubic, leaveTangent=-143.9088, tangentMode=RCTM_User, time=0.1, value=50}
         {arriveTangent=-168.5692, interpMode=RCIM_Cubic, leaveTangent=-352.3025, tangentMode=RCTM_Break, time=0.35, value=17}
         {arriveTangent=-9.0214, interpMode=RCIM_Cubic, leaveTangent=-9.0214, tangentMode=RCTM_User, time=0.45, value=1}
         <at 0.6 0 RCIM_Cubic>
        ]}}})
    Target Position Value:NiagaraPosition=$Output.SolveForcesAndVelocity.Position
    Variable Type=::NewEnumerator6
   ) {enabled=false}
   /SetVariables(Particles.Position=&AddVectorToPosition(
     Position=$StackContext.InterpolateOverTime.InterpolateOverTime_Position.Moving Average
     Vector=&Multiply_Vector(
      A=&MakeVector(
       X=&FloatFromCurve(
        FloatCurve=DataInterface() {propertyValues {Curve {
           defaultValue=1
           keys [
            {arriveTangent=3.5337, interpMode=RCIM_Cubic, leaveTangent=3.5337, tangentMode=RCTM_User, time=0, value=0}
            {interpMode=RCIM_Cubic, tangentMode=RCTM_User, time=1, value=1}
           ]
          }}}
        Scale Curve=&RandomRangeFloat(Minimum=-2, Maximum=2)
       )
       Y=&FloatFromCurve(
        FloatCurve=DataInterface() {propertyValues {Curve {
           defaultValue=1
           keys [
            {arriveTangent=3.5337, interpMode=RCIM_Cubic, leaveTangent=3.5337, tangentMode=RCTM_User, time=0, value=0}
            {interpMode=RCIM_Cubic, tangentMode=RCTM_User, time=1, value=1}
           ]
          }}}
        Scale Curve=&RandomRangeFloat(Minimum=-2, Maximum=2)
       )
       Z=&FloatFromCurve(
        FloatCurve=DataInterface() {propertyValues {Curve {
           defaultValue=1
           keys [
            {arriveTangent=3.5337, interpMode=RCIM_Cubic, leaveTangent=3.5337, tangentMode=RCTM_User, time=0, value=0}
            {interpMode=RCIM_Cubic, tangentMode=RCTM_User, time=1, value=1}
           ]
          }}}
        Scale Curve=&RandomRangeFloat(Minimum=-2, Maximum=2)
       )
      )
      B=&VectorFromFloat(Value=&RandomRangeFloat(Minimum=1, Maximum=2.25))
     )
    )) {enabled=false, isSetParameters=true, module=SetVariables_#5}
  ]
  rendererClasses [&NiagaraSpriteRendererProperties]
  renderers [{rendererClass=&NiagaraSpriteRendererProperties}]
 }
 NE_MicRibbon {
  emitterSpawnScript [/SetVariables(Emitter.ParticlePerRibbon=15, Emitter.RibbonNumber=1) {isSetParameters=true, module=SetVariables_#6}]
  emitterUpdateScript [
   &EmitterState(
    Inactive Response=::NewEnumerator1
    Life Cycle Mode=::NewEnumerator1
    Loop Behavior=::NewEnumerator1
    Loop Duration Mode=::NewEnumerator1
    UseLoopDelay=false
   )
   &SpawnBurst_Instantaneous(
    Spawn Count=&Multiply_Int(A=$Emitter.ParticlePerRibbon, B=$Emitter.RibbonNumber)
    Use Spawn Probability=false
    UseLoopCountLimit=false
   )
  ]
  particleSpawnScript [
   &InitializeParticle(
    Color=<rgba 1 1 1 0.9>
    Lifetime=5
    Ribbon Width Mode=::NewEnumerator1
    Sprite Size Mode=::NewEnumerator3
    UsePositionOffset=false
    Write Material Random=true
   )
   &PartitionParticles(
    ID to Partition:NiagaraInt32=&ReturnExecIndex()
    Number Of Partitions=$Emitter.RibbonNumber
    Particles Per Partition:NiagaraInt32=$Emitter.ParticlePerRibbon
    Partition Mode:ENiagara_PartitionMode_v2=::NewEnumerator1
   )
   /SetVariables(Particles.NormalizedRibbonPos:NiagaraFloat=&OneMinusFloat(Float=$Particles.PartitionParticles.Normalized Index In Partition)) {isSetParameters=true, module=SetVariables_#7}
   /SetVariables(
    Particles.SpiralFreq=0.8
    Particles.SpiralStrength:NiagaraFloat=&Multiply_Float(
     A=15
     B=&FloatFromCurve(
      FloatCurve=DataInterface() {propertyValues {Curve {keys [
          {arriveTangent=10.2716, interpMode=RCIM_Cubic, leaveTangent=10.2716, tangentMode=RCTM_User, time=0, value=-0.025}
          {interpMode=RCIM_Cubic, tangentMode=RCTM_User, time=0.2, value=1}
          {arriveTangent=-0.504, interpMode=RCIM_Cubic, leaveTangent=-0.504, tangentMode=RCTM_User, time=0.6, value=0.95}
          {arriveTangent=-0.1348, interpMode=RCIM_Cubic, leaveTangent=-0.1348, tangentMode=RCTM_User, time=1, value=0}
         ]}}}
      CurveIndex=$Particles.NormalizedRibbonPos
     )
    )
   ) {isSetParameters=true, module=SetVariables_#8}
   /SetVariables(Particles.RibbonU0Override=&OneMinusFloat(Float=$Particles.NormalizedRibbonPos)) {isSetParameters=true, module=SetVariables_#9}
   /SetVariables() {isSetParameters=true, module=SetVariables_#10}
  ]
  particleUpdateScript [
   &ParticleState(Kill Particles When Lifetime Has Elapsed=false)
   &SkeletalMeshLocation(
    MaskSpawnGroup=false
    Mesh Sampling Type=::NewEnumerator5
    Position Sampling=::NewEnumerator2
    Skeletal Mesh=$System.Skeletal Mesh
    Socket Sampling Mode=::NewEnumerator5
    Use Sampled Mesh Attribute as Ribbon ID=false
   )
   /SetVariables(Particles.GoalPos:NiagaraPosition=&AddVectorToPosition(
     Position=$Particles.SkeletalMeshLocation.SampledPosition
     Vector=&Lerp_Vector(A=<0 17.0783 3>, B=<0 -14 7>, Alpha=$Particles.NormalizedRibbonPos)
    )) {isSetParameters=true, module=SetVariables_#11}
   /SetVariables(Particles.SpiralStrength:NiagaraFloat=&Multiply_Float(
     A=&Multiply_Float(
      A=2
      B=&Waveform(
       Global Phase=&Add_Float(A=$Emitter.Age, B=$Particles.NormalizedRibbonPos)
       "Waveform [1]"=::NewEnumerator12
       "[1] Frequency"=0.5
       "[1] Amplitude Min/Max"=<1 1.25>
      )
     )
     B=&FloatFromCurve(
      FloatCurve=DataInterface() {propertyValues {Curve {keys [
          {arriveTangent=3.1146, interpMode=RCIM_Cubic, leaveTangent=3.1146, leaveTangentWeight=6.4519, tangentMode=RCTM_User, time=0, value=0.2}
          {arriveTangent=1.7735, interpMode=RCIM_Cubic, leaveTangent=1.7735, tangentMode=RCTM_User, time=0.5, value=1.25}
          {arriveTangentWeight=0.2147, interpMode=RCIM_Cubic, leaveTangentWeight=0.1431, tangentMode=RCTM_User, time=0.8, value=1.75}
          {arriveTangent=-19.8738, interpMode=RCIM_Cubic, leaveTangent=-19.8738, tangentMode=RCTM_User, time=1, value=0.3}
         ]}}}
      CurveIndex=$Particles.NormalizedRibbonPos
     )
    )) {isSetParameters=true, module=SetVariables_#12}
   /SetVariables(Particles.SpiralTime:NiagaraFloat=&Add_Float(
     A=&Add_Float(A=&Multiply_Float(A=$System.Age, B=0.1), B=$Particles.NormalizedRibbonPos)
     B=&Multiply_FloatByInt(
      Float=&Divide_Float(A=$Particles.SpiralFreq, B=&MakeFloatFromInt(Int=$Emitter.RibbonNumber))
      Integer=$Particles.PartitionParticles.Partition
     )
    )) {isSetParameters=true, module=SetVariables_#13}
   /SetVariables(Particles.SpiralOffset:Vector3f=&MakeVector(
     X=&Sine(
      Normalized Angle=$Particles.SpiralTime
      Period=$Particles.SpiralFreq
      Scale=&Multiply_Float(A=$Particles.SpiralStrength, B=1.5)
      Bias=0
     )
     Y=0
     Z=&Cosine(
      Normalized Angle=$Particles.SpiralTime
      Period=$Particles.SpiralFreq
      Scale=&Multiply_Float(A=$Particles.SpiralStrength, B=1.5)
      Bias=0
     )
    )) {isSetParameters=true, module=SetVariables_#14}
   &Color(Color=&MultiplyLinearColor(
     Linear Color=&ColorFromCurve(
      ColorCurve=DataInterface() {propertyValues {
        AlphaCurve {keys [<at 0.0365 1 RCIM_Cubic>,<at 0.7167 1>,<at 1 0>]}
        BlueCurve {keys [<at 0.0057 1>,<at 0.7608 1.5>]}
        GreenCurve {keys [<at 0.0057 2>,<at 0.7608 1.5>]}
        RedCurve {keys [<at 0.0057 0.5>,<at 0.7608 0.375>]}
       }}
      CurveIndex=$Particles.NormalizedRibbonPos
     )
     Scale Factor=<1 1 1 1>
    )) {module=Color001}
   /SetVariables(Particles.Position=&AddVectorToPosition(Position=$Particles.GoalPos, Vector=$Particles.SpiralOffset)) {isSetParameters=true, module=SetVariables_#15}
   &InheritSourceMovement(
    Applied Position Falloff Distance=275
    Applied Position Scale=&VectorFromFloat(Value=&Multiply_Float(
      A=-0.5
      B=&FloatFromCurve(
       FloatCurve=DataInterface() {propertyValues {Curve {keys [
           {arriveTangent=0.9742, interpMode=RCIM_Cubic, leaveTangent=0.9742, tangentMode=RCTM_User, time=0, value=0}
           <at 0.5 0.3125 RCIM_Cubic>
           {arriveTangent=-0.7756, interpMode=RCIM_Cubic, leaveTangent=-0.7756, tangentMode=RCTM_User, time=1, value=0}
          ]}}}
       CurveIndex=$Particles.NormalizedRibbonPos
      )
     ))
    ApplyForce=false
    ApplyPosition=true
    ApplyPositionFalloff=true
    ApplyPositionFalloffCurve=false
    Source Velocity:Vector3f=&TransformVector(Vector=$Engine.Owner.Velocity, Source Space=::World, Destination Space=::Local)
   )
   &SolveForcesAndVelocity(Clamp Velocity=false, Limit Acceleration=false)
   /SetVariables(Particles.RibbonFacing=&NormalizeVector(
     Vector To Normalize=&TransformVector(Vector=$Particles.SpiralOffset, Destination Space=::Simulation)
     Safe Normalize=true
     Safe Fallback Threshold=0
     Safe Fallback Vector=<0 0 0.001>
    )) {isSetParameters=true, module=SetVariables_#16}
   &ScaleRibbonWidth(Ribbon Width Scale=&Multiply_Float(
     A=&FloatFromCurve(
      FloatCurve=DataInterface() {propertyValues {Curve {keys [
          {arriveTangent=0.5186, interpMode=RCIM_Cubic, leaveTangent=0.5186, leaveTangentWeight=3.2425, tangentMode=RCTM_User, time=0, value=0}
          {arriveTangentWeight=0.2872, interpMode=RCIM_Cubic, leaveTangentWeight=0.1567, tangentMode=RCTM_User, time=0.6, value=1}
          {arriveTangentWeight=0.2515, interpMode=RCIM_Cubic, tangentMode=RCTM_User, time=1, value=0}
         ]}}}
      CurveIndex=$Particles.NormalizedRibbonPos
     )
     B=&Waveform(
      Global Phase=&Add_Float(A=$Emitter.Age, B=$Particles.NormalizedRibbonPos)
      "Waveform [1]"=::NewEnumerator12
      "[1] Frequency"=0.5
      "[1] Amplitude Min/Max"=<0.9 1.25>
     )
    ))
   /SetVariables(Particles.GlowWidth:NiagaraFloat=&Multiply_Float(A=$Particles.RibbonWidth, B=1.3)) {isSetParameters=true, module=SetVariables_#17}
   /SetVariables(Particles.GlowColor=&MultiplyLinearColor(
     Linear Color=&MultiplyLinearColor(Linear Color=$Particles.Color, Scale Factor=&MakeVector4_FromLinearColor(LinearColor=<rgba 1 1 1 1>))
     Scale Factor=&MakeVector4(
      X=1
      Y=1
      Z=1
      W=&FloatFromCurve(
       FloatCurve=DataInterface() {propertyValues {Curve {keys [
           {arriveTangent=8.2127, interpMode=RCIM_Cubic, leaveTangent=8.2127, tangentMode=RCTM_User, time=0, value=0}
           {arriveTangent=0.4596, arriveTangentWeight=0.232, leaveTangentWeight=2.2538, time=0.2, value=1}
           {arriveTangent=0.0486, interpMode=RCIM_Cubic, leaveTangent=0.0486, tangentMode=RCTM_User, time=0.75, value=1}
           {arriveTangent=-4.4982, interpMode=RCIM_Cubic, leaveTangent=-4.4982, tangentMode=RCTM_User, time=1, value=0}
          ]}}}
       CurveIndex=$Particles.NormalizedRibbonPos
       Scale Curve=&OneMinusFloat(Float=$System.Timeline.NormalizedTime)
      )
     )
    )) {isSetParameters=true, module=SetVariables_#18}
   &DynamicMaterialParameters(
    Index 0 Param 1=1
    Index 1 Param 2=1
    Index 1 Param 3=&FloatFromCurve(
     FloatCurve=DataInterface() {propertyValues {Curve {keys [
         {arriveTangent=-2.6187, interpMode=RCIM_Cubic, leaveTangent=-2.6187, tangentMode=RCTM_User, time=0, value=1.5}
         {arriveTangent=-0.275, interpMode=RCIM_Cubic, leaveTangent=-0.275, time=0.7, value=0}
         {interpMode=RCIM_Cubic, tangentMode=RCTM_User, time=1, value=1.225}
        ]}}}
     CurveIndex=$Particles.PartitionParticles.Normalized Index In Partition
     Scale Curve=0.5
    )
    Module.Index0Param0Write=true
    Module.Index0Param1Write=true
    Module.Index0Param2Write=true
    Module.Index0Param3Write=true
    Module.Index1Param0Write=true
    Module.Index1Param1Write=true
    Module.Index1Param2Write=true
    Module.Index1Param3Write=true
    Write Parameter Index 1=true
   )
   &ScaleColor(
    Curve Index=$Particles.NormalizedRibbonPos
    Linear Color Curve=DataInterface() {propertyValues {
      AlphaCurve {keys [<at 0 0>, {arriveTangent=1.4043, interpMode=RCIM_Cubic, time=0.1636, value=1}, <at 0.7167 1>, <at 1 0>]}
      BlueCurve {keys [<at 0.0057 1>,<at 0.7608 1.5>]}
      GreenCurve {keys [<at 0.0057 1>,<at 0.7608 1.5>]}
      RedCurve {keys [<at 0.0057 0.5029>,<at 0.7608 0.375>]}
     }}
    Scale Mode=::NewEnumerator2
   ) {module=ScaleColor001}
   &ScaleColor(
    Scale Alpha=&FloatFromCurve(
     FloatCurve=DataInterface() {propertyValues {Curve {keys [
         {arriveTangent=8.2127, interpMode=RCIM_Cubic, leaveTangent=8.2127, tangentMode=RCTM_User, time=0, value=0}
         {arriveTangent=0.4596, arriveTangentWeight=0.232, leaveTangentWeight=2.2538, time=0.2, value=1}
         <at 0.3 0.75>
         <at 0.5 0.95>
         <at 0.6 0.75>
         {arriveTangent=0.0486, interpMode=RCIM_Cubic, leaveTangent=0.0486, tangentMode=RCTM_User, time=0.75, value=1}
         {arriveTangent=-4.4982, interpMode=RCIM_Cubic, leaveTangent=-4.4982, tangentMode=RCTM_User, time=1, value=0}
        ]}}}
     CurveIndex=$Particles.NormalizedRibbonPos
     Scale Curve=&OneMinusFloat(Float=$System.Timeline.NormalizedTime)
    )
    ScaleA=true
    ScaleRGB=true
   ) {enabled=false}
   &ScaleColor(
    Scale RGB=&MakeVectorFromLinearColorRGB(Color=&MultiplyLinearColor(
      Linear Color=$StackContext.Color
      Scale Factor=&MakeVector4_FromFloat(Float=&FNTimeOfDayScalar(Night Scale=1, Morning Scale=1, Evening Scale=1.05))
     ))
    ScaleA=true
    ScaleRGB=true
   ) {module=ScaleColor002}
  ]
  rendererClasses [&NiagaraRibbonRendererProperties]
  renderers [
   {rendererClass=&NiagaraRibbonRendererProperties}
   {rendererClass=&NiagaraRibbonRendererProperties, rendererIndex=1}
  ]
 }
 NE_Motes {
  emitterUpdateScript [
   &EmitterState()
   &SpawnRate(Spawn Probability=0.25, SpawnRate=3, Use Spawn Probability=false)
  ]
  particleSpawnScript [
   &PartitionParticles(
    ID to Partition:NiagaraInt32=&ReturnExecIndex()
    Number Of Partitions=$System.RibbonNumber
    Partition Mode:ENiagara_PartitionMode=::NewEnumerator5
   )
   /SetVariables(Particles.VisibilityTag=&RandomRangeInt(Minimum=0, Maximum=1)) {enabled=false, isSetParameters=true}
   &InitializeParticle(
    Color=&ColorFromCurve(
     ColorCurve=DataInterface() {propertyValues {
       AlphaCurve {keys [<at 0 0.8 RCIM_Cubic>]}
       BlueCurve {keys [<at 0 0.3631 RCIM_Cubic>,<at 0.9825 0.983 RCIM_Cubic>]}
       GreenCurve {keys [<at 0 0.5104 RCIM_Cubic>,<at 0.9825 1 RCIM_Cubic>]}
       RedCurve {keys [<at 0 0.0136 RCIM_Cubic>,<at 0.9825 0.2396 RCIM_Cubic>]}
      }}
     CurveIndex=&RandomRangeFloat(Minimum=0, Maximum=1)
    )
    Lifetime=&RandomRangeFloat(Minimum=5, Maximum=7)
    Sprite Size Mode=::NewEnumerator3
    Sprite UV Mode=::NewEnumerator3
    Uniform Sprite Size=&RandomRangeFloat(Minimum=0.5, Maximum=0.75)
    UsePositionOffset=true
    Write Material Random=true
   )
   &ShapeLocation(
    Apply To Particle Position=::NewEnumerator1
    MaskSpawnGroup=false
    Offset Mode=::NewEnumerator0
    Override Seed=false
    Sphere Radius=20
    UseOwnerScale=true
   )
   /SetVariables() {isSetParameters=true}
   &CameraOffset(Camera Offset Amount=10)
  ]
  particleUpdateScript [
   &ParticleState()
   &SkeletalMeshLocation(
    Bone / Socket Index=3
    MaskSpawnGroup=false
    Mesh Sampling Type=::NewEnumerator5
    Position Sampling=::NewEnumerator2
    Sampled Position Offset=<0 0 15>
    Skeletal Mesh=$System.Skeletal Mesh
    Socket Sampling Mode=::NewEnumerator5
    Use Sampled Mesh Attribute as Ribbon ID=false
   )
   &self:FollowRibbon_01(
    Alpha:NiagaraFloat=&FloatFromCurve(FloatCurve=DataInterface() {propertyValues {Curve {keys [<at 0.1 1>,<at 1 0>]}}})
    ParticlePerRibbon:NiagaraInt32=$System.ParticlePerRibbon
    RibbonNumber:NiagaraInt32=$Particles.PartitionParticles.Partition
    RibbonReader=DataInterface() {propertyValues {EmitterBinding {bindingMode=Other, emitterName=NE_SpiralRibbon001}}}
   ) {module=FollowRibbon_01}
   /SetVariables(Particles.Position=&Lerp_Vector(
     A=Linked() {linkedVariable {name=Particles.Position, type=NiagaraPosition}}
     B=&AddVectorToPosition(
      Position=&AddVectorToPosition(
       Position=$Output.FollowRibbon_01.RibbonPos
       Vector=&RandomRangeVector(Minimum=<-8 -8 0>, Maximum=<8 8 3>)
      )
      Vector=&Multiply_Vector(A=$Output.FollowRibbon_01.DirectionVector, B=&RandomRangeVector(Minimum=<-5 -5 0>, Maximum=<5 5 0>))
     )
     Alpha=&Multiply_Float(A=$Engine.DeltaTime, B=8)
    )) {isSetParameters=true}
   /SetVariables(Particles.Velocity=&Multiply_VectorByFloat(Vector=$Output.FollowRibbon_01.DirectionVector, Float=-1)) {isSetParameters=true}
   &ScaleSpriteSize(
    Uniform Curve Scale=&OneMinusFloat(Float=$System.Timeline.NormalizedTime)
    Uniform Curve Sprite Scale=DataInterface() {propertyValues {Curve {keys [
        {arriveTangent=8.2127, interpMode=RCIM_Cubic, leaveTangent=8.2127, tangentMode=RCTM_User, time=0.05, value=0}
        {
         arriveTangent=0.4596
         arriveTangentWeight=0.232
         interpMode=RCIM_Cubic
         leaveTangent=0.4596
         leaveTangentWeight=2.2538
         tangentMode=RCTM_User
         time=0.2
         value=1
        }
        <at 0.4 1 RCIM_Cubic>
        {arriveTangent=0.0486, interpMode=RCIM_Cubic, leaveTangent=0.0486, tangentMode=RCTM_User, time=0.75, value=1}
        {arriveTangent=-4.4982, interpMode=RCIM_Cubic, leaveTangent=-4.4982, tangentMode=RCTM_User, time=1, value=0}
       ]}}}
   ) {enabled=false}
   &ScaleSpriteSize(
    Uniform Curve Scale=&Waveform("Waveform [1]"=::NewEnumerator0, "[1] Frequency"=0.5, "[1] Amplitude Min/Max"=<0.9 1.2>)
    Uniform Curve Sprite Scale=DataInterface() {propertyValues {Curve {keys [
        {arriveTangent=8.2127, interpMode=RCIM_Cubic, leaveTangent=8.2127, tangentMode=RCTM_User, time=0.05, value=0}
        {
         arriveTangent=0.4596
         arriveTangentWeight=0.232
         interpMode=RCIM_Cubic
         leaveTangent=0.4596
         leaveTangentWeight=2.2538
         tangentMode=RCTM_User
         time=0.2
         value=1
        }
        <at 0.4 1 RCIM_Cubic>
        {arriveTangent=0.0486, interpMode=RCIM_Cubic, leaveTangent=0.0486, tangentMode=RCTM_User, time=0.75, value=1}
        {arriveTangent=-4.4982, interpMode=RCIM_Cubic, leaveTangent=-4.4982, tangentMode=RCTM_User, time=1, value=0}
       ]}}}
   ) {module=ScaleSpriteSize002}
   &ScaleSpriteSize(
    Scale Factor=&Multiply_Vector2D_ByFloat(Float=2, Vector2D=&RandomRangeVector2D(Minimum=<1 2>, Maximum=<1 2.5>))
    Scale Sprite Size Mode=::NewEnumerator2
   ) {module=ScaleSpriteSize001}
   &ScaleColor(
    Scale Alpha=&FloatFromCurve(FloatCurve=DataInterface() {propertyValues {Curve {
        defaultValue=1
        keys [
         {interpMode=RCIM_Cubic, tangentMode=RCTM_User, time=0, value=0}
         {interpMode=RCIM_Cubic, tangentMode=RCTM_User, time=0.2, value=1}
         {arriveTangentWeight=0.3202, interpMode=RCIM_Cubic, leaveTangentWeight=0.1067, tangentMode=RCTM_User, time=0.8, value=1}
         {tangentMode=RCTM_User, time=1, value=0}
        ]
       }}})
    ScaleA=true
    ScaleRGB=true
   )
   &DynamicMaterialParameters(
    Index 0 Param 3=1
    Index 0 Param 4=&MaskFloatBySpawnGroup(Masked Float=0, Passthrough Float=0)
    Module.Param0WriteEnabled=true
    Module.Param1WriteEnabled=true
    Module.Param2WriteEnabled=true
    Module.Param3WriteEnabled=true
   ) {enabled=false}
   /Niagara/Modules/Update/SubUV/V2/SubUVAnimation(
    Evaluation Type=::NewEnumerator0
    "Sprite Renderer "=DataInterface() {propertyValues {SpriteRenderer=&self:NE_Motes_0.NiagaraSpriteRendererProperties_0}}
    SubUV Animation Mode:ENiagara_SubUVLookupModeV2=::NewEnumerator2
    UseEndFrame=false
    UseStartFrame=false
   )
   /SetVariables(Particles.glowsize:Vector2f=&Multiply_Vector2D_ByFloat(Float=2, Vector2D=$Particles.SpriteSize)) {isSetParameters=true}
  ]
  rendererClasses [&NiagaraSpriteRendererProperties]
  renderers [
   {rendererClass=&NiagaraSpriteRendererProperties}
   {rendererClass=&NiagaraSpriteRendererProperties, rendererIndex=1}
  ]
 }
 NE_SpiralRibbon001 {
  emitterSpawnScript [/SetVariables(Emitter.ParticlePerRibbon:NiagaraInt32=$System.ParticlePerRibbon, Emitter.RibbonNumber:NiagaraInt32=$System.RibbonNumber) {isSetParameters=true, module=SetVariables_#6}]
  emitterUpdateScript [
   &EmitterState(
    Inactive Response=::NewEnumerator1
    Life Cycle Mode=::NewEnumerator1
    Loop Behavior=::NewEnumerator1
    Loop Duration Mode=::NewEnumerator1
    UseLoopDelay=false
   )
   &SpawnBurst_Instantaneous(
    Spawn Count=&Multiply_Int(A=$Emitter.ParticlePerRibbon, B=$Emitter.RibbonNumber)
    Use Spawn Probability=false
    UseLoopCountLimit=false
   )
  ]
  particleSpawnScript [
   &InitializeParticle(Lifetime=5, Ribbon Width=8, Ribbon Width Mode=::NewEnumerator1, UsePositionOffset=false, Write Material Random=true)
   &PartitionParticles(
    ID to Partition:NiagaraInt32=&ReturnExecIndex()
    Number Of Partitions=$Emitter.RibbonNumber
    Particles Per Partition:NiagaraInt32=$Emitter.ParticlePerRibbon
    Partition Mode:ENiagara_PartitionMode_v2=::NewEnumerator1
   )
   /SetVariables(Particles.NormalizedRibbonPos:NiagaraFloat=&OneMinusFloat(Float=$Particles.PartitionParticles.Normalized Index In Partition)) {isSetParameters=true, module=SetVariables_#7}
   /SetVariables(
    Particles.SpiralFreq=0.6
    Particles.SpiralFreq2:NiagaraFloat=&MakeCustomFloatFromBool(Bool=&SetBoolByFloatComparison(A=$User.IsPickaxe, B=0), True Float=0.7, False Float=0.5)
    Particles.SpiralFreq3=0.6
    Particles.SpiralStrength:NiagaraFloat=&Multiply_Float(
     A=15
     B=&FloatFromCurve(
      FloatCurve=DataInterface() {propertyValues {Curve {keys [
          {arriveTangent=10.2716, interpMode=RCIM_Cubic, leaveTangent=10.2716, tangentMode=RCTM_User, time=0, value=-0.025}
          {interpMode=RCIM_Cubic, tangentMode=RCTM_User, time=0.2, value=1}
          {arriveTangent=-0.504, interpMode=RCIM_Cubic, leaveTangent=-0.504, tangentMode=RCTM_User, time=0.6, value=0.95}
          {arriveTangent=-0.1348, interpMode=RCIM_Cubic, leaveTangent=-0.1348, tangentMode=RCTM_User, time=1, value=0}
         ]}}}
      CurveIndex=$Particles.NormalizedRibbonPos
     )
    )
    Particles.SpiralStrength3:NiagaraFloat=&Multiply_Float(
     A=15
     B=&FloatFromCurve(
      FloatCurve=DataInterface() {propertyValues {Curve {keys [
          {arriveTangent=10.2716, interpMode=RCIM_Cubic, leaveTangent=10.2716, tangentMode=RCTM_User, time=0, value=-0.025}
          {interpMode=RCIM_Cubic, tangentMode=RCTM_User, time=0.2, value=1}
          {arriveTangent=-0.504, interpMode=RCIM_Cubic, leaveTangent=-0.504, tangentMode=RCTM_User, time=0.6, value=0.95}
          {arriveTangent=-0.1348, interpMode=RCIM_Cubic, leaveTangent=-0.1348, tangentMode=RCTM_User, time=1, value=0}
         ]}}}
      CurveIndex=$Particles.NormalizedRibbonPos
     )
    )
   ) {isSetParameters=true, module=SetVariables_#8}
   /SetVariables(Particles.RibbonU0Override=&OneMinusFloat(Float=$Particles.NormalizedRibbonPos)) {isSetParameters=true, module=SetVariables_#9}
   /SetVariables() {isSetParameters=true, module=SetVariables_#10}
   /SetVariables(Particles.RibbonWidth2=11, Particles.RibbonWidth3=5) {isSetParameters=true}
   /SetVariables(Particles.Color2=<rgba 1 1 1 0.6>) {isSetParameters=true}
  ]
  particleUpdateScript [
   &ParticleState(Kill Particles When Lifetime Has Elapsed=false)
   &SkeletalMeshLocation(
    Bone / Socket Index=3
    MaskSpawnGroup=false
    Mesh Sampling Type=::NewEnumerator5
    Position Sampling=::NewEnumerator2
    Skeletal Mesh=$System.Skeletal Mesh
    Socket Sampling Mode=::NewEnumerator5
    Use Sampled Mesh Attribute as Ribbon ID=false
   )
   /SetVariables(
    Particles.GoalPos:NiagaraPosition=&AddVectorToPosition(
     Position=$Particles.SkeletalMeshLocation.SampledPosition
     Vector=&Lerp_Vector(A=<0 0 -14>, B=<0 0 137>, Alpha=$Particles.NormalizedRibbonPos)
    )
    Particles.GoalPos2:NiagaraPosition=&AddVectorToPosition(
     Position=$Particles.SkeletalMeshLocation.SampledPosition
     Vector=&MakeCustomVectorFromBool(
      Bool=&SetBoolByFloatComparison(A=$User.IsPickaxe, B=0)
      True Vector=&VectorFromCurve(
       VectorCurve=DataInterface() {propertyValues {
         XCurve {keys [<at 0 0>,<at 1 0>]}
         YCurve {keys [<at 0 0>,<at 1 0>]}
         ZCurve {keys [
           {arriveTangent=221.3778, interpMode=RCIM_Cubic, leaveTangent=221.3778, tangentMode=RCTM_User, time=0, value=0}
           {arriveTangent=14, interpMode=RCIM_Cubic, leaveTangent=14, time=0.2, value=-2.5}
           {arriveTangent=38.3399, interpMode=RCIM_Cubic, leaveTangent=38.3399, tangentMode=RCTM_User, time=0.5, value=7}
           {arriveTangent=125.412, tangentMode=RCTM_User, time=1, value=58}
          ]}
        }}
       CurveIndex=$Particles.NormalizedRibbonPos
      )
      False Vector=&VectorFromCurve(
       VectorCurve=DataInterface() {propertyValues {
         XCurve {keys [<at 0 0>,<at 1 0>]}
         YCurve {keys [<at 0 0>,<at 1 0>]}
         ZCurve {keys [
           {arriveTangent=147.4484, interpMode=RCIM_Cubic, leaveTangent=147.4484, tangentMode=RCTM_User, time=0, value=0}
           <at 0.2 -2.5 RCIM_Cubic>
           {arriveTangent=15.1148, interpMode=RCIM_Cubic, leaveTangent=15.1148, tangentMode=RCTM_User, time=0.4, value=0}
           {arriveTangent=143.5554, tangentMode=RCTM_User, time=1, value=45}
          ]}
        }}
       CurveIndex=$Particles.NormalizedRibbonPos
      )
     )
    )
    Particles.GoalPos3:NiagaraPosition=&AddVectorToPosition(
     Position=$Particles.SkeletalMeshLocation.SampledPosition
     Vector=&Lerp_Vector(A=<0 0 -14>, B=<0 0 120>, Alpha=$Particles.NormalizedRibbonPos)
    )
   ) {isSetParameters=true, module=SetVariables_#11}
   /SetVariables(
    Particles.SpiralStrength:NiagaraFloat=&Multiply_Float(
     A=3
     B=&FloatFromCurve(
      FloatCurve=DataInterface() {propertyValues {Curve {keys [
          {arriveTangent=48.3786, interpMode=RCIM_Cubic, leaveTangent=48.3786, tangentMode=RCTM_User, time=0.1, value=0}
          {arriveTangent=0.0033, interpMode=RCIM_Cubic, leaveTangent=0.0033, tangentMode=RCTM_User, time=0.2, value=1.625}
          {arriveTangent=-1.4093, interpMode=RCIM_Cubic, leaveTangent=-1.4093, tangentMode=RCTM_User, time=0.6, value=1.5}
          {arriveTangent=-0.6389, interpMode=RCIM_Cubic, leaveTangent=-0.6389, tangentMode=RCTM_User, time=1, value=0.2}
         ]}}}
      CurveIndex=$Particles.NormalizedRibbonPos
     )
    )
    Particles.SpiralStrength2:NiagaraFloat=&Multiply_Float(
     A=&MakeCustomFloatFromBool(Bool=&SetBoolByFloatComparison(A=$User.IsPickaxe, B=0), True Float=7.5, False Float=9)
     B=&FloatFromCurve(
      FloatCurve=DataInterface() {propertyValues {Curve {
         defaultValue=1.625
         keys [
          {arriveTangent=-0.3372, interpMode=RCIM_Cubic, leaveTangent=-0.3372, tangentMode=RCTM_User, time=0, value=1.45}
          {arriveTangentWeight=0.2963, interpMode=RCIM_Cubic, leaveTangentWeight=0.127, tangentMode=RCTM_User, time=0.7, value=1.35}
          {arriveTangent=-0.3467, interpMode=RCIM_Cubic, leaveTangent=-0.3467, tangentMode=RCTM_User, time=1, value=0}
         ]
        }}}
      CurveIndex=$Particles.NormalizedRibbonPos
     )
    )
    Particles.SpiralStrength3:NiagaraFloat=&Multiply_Float(
     A=&Waveform("Waveform [1]"=::NewEnumerator0, "[1] Phase Scale"=2, "[1] Frequency"=0.2, "[1] Amplitude Min/Max"=<5.5 7>)
     B=&FloatFromCurve(
      FloatCurve=DataInterface() {propertyValues {Curve {keys [
          {arriveTangent=0.4644, interpMode=RCIM_Cubic, leaveTangent=0.4644, tangentMode=RCTM_User, time=0.1, value=0.275}
          {arriveTangent=0.5085, interpMode=RCIM_Cubic, leaveTangent=0.5085, tangentMode=RCTM_User, time=0.25, value=1.7}
          {arriveTangent=-2.4336, interpMode=RCIM_Cubic, leaveTangent=-2.4336, tangentMode=RCTM_User, time=0.6, value=1.35}
          {arriveTangent=-2.6365, interpMode=RCIM_Cubic, leaveTangent=-2.6365, tangentMode=RCTM_User, time=1, value=0.2}
         ]}}}
      CurveIndex=$Particles.NormalizedRibbonPos
     )
    )
   ) {isSetParameters=true, module=SetVariables_#12}
   /SetVariables(
    Particles.SpiralTime:NiagaraFloat=&Add_Float(
     A=&Add_Float(A=&Multiply_Float(A=$System.Age, B=0.09), B=$Particles.NormalizedRibbonPos)
     B=&Multiply_FloatByInt(
      Float=&Divide_Float(A=$Particles.SpiralFreq, B=&MakeFloatFromInt(Int=$Emitter.RibbonNumber))
      Integer=$Particles.PartitionParticles.Partition
     )
    )
    Particles.SpiralTime2:NiagaraFloat=&Add_Float(
     A=&Add_Float(A=&Multiply_Float(A=$System.Age, B=0.015), B=$Particles.NormalizedRibbonPos)
     B=&Multiply_FloatByInt(
      Float=&Divide_Float(A=$Particles.SpiralFreq2, B=&MakeFloatFromInt(Int=$Emitter.RibbonNumber))
      Integer=$Particles.PartitionParticles.Partition
     )
    )
    Particles.SpiralTime3:NiagaraFloat=&Add_Float(
     A=&Add_Float(A=&Multiply_Float(A=$System.Age, B=0.09), B=$Particles.NormalizedRibbonPos)
     B=&Multiply_FloatByInt(
      Float=&Divide_Float(A=$Particles.SpiralFreq3, B=&MakeFloatFromInt(Int=$Emitter.RibbonNumber))
      Integer=$Particles.PartitionParticles.Partition
     )
    )
   ) {isSetParameters=true, module=SetVariables_#13}
   /SetVariables(
    Particles.SpiralOffset:Vector3f=&MakeVector(
     X=&Sine(
      Normalized Angle=$Particles.SpiralTime
      Period=$Particles.SpiralFreq
      Scale=&Multiply_Float(A=$Particles.SpiralStrength, B=1.5)
      Bias=0
     )
     Y=&Cosine(
      Normalized Angle=$Particles.SpiralTime
      Period=$Particles.SpiralFreq
      Scale=&Multiply_Float(A=$Particles.SpiralStrength, B=1.5)
      Bias=0
     )
     Z=0
    )
    Particles.SpiralOffset2:Vector3f=&MakeVector(
     X=&Sine(
      Normalized Angle=$Particles.SpiralTime2
      Period=$Particles.SpiralFreq2
      Scale=&Multiply_Float(A=$Particles.SpiralStrength2, B=1.5)
      Bias=0
     )
     Y=&Cosine(
      Normalized Angle=$Particles.SpiralTime2
      Period=$Particles.SpiralFreq2
      Scale=&Multiply_Float(A=$Particles.SpiralStrength2, B=1.5)
      Bias=0
     )
     Z=0
    )
    Particles.SpiralOffset3:Vector3f=&MakeVector(
     X=&Sine(
      Normalized Angle=$Particles.SpiralTime3
      Period=$Particles.SpiralFreq3
      Scale=&Multiply_Float(A=$Particles.SpiralStrength3, B=1.5)
      Bias=0.2
     )
     Y=&Cosine(
      Normalized Angle=$Particles.SpiralTime3
      Period=$Particles.SpiralFreq3
      Scale=&Multiply_Float(A=$Particles.SpiralStrength3, B=1.5)
      Bias=0.2
     )
     Z=0
    )
   ) {isSetParameters=true, module=SetVariables_#14}
   &Color(Color=&MultiplyLinearColor(
     Linear Color=&ColorFromCurve(
      ColorCurve=DataInterface() {propertyValues {
        AlphaCurve {keys [<at 0.0365 1 RCIM_Cubic>,<at 0.7167 1>,<at 1 0>]}
        BlueCurve {keys [<at 0.0057 1>,<at 0.7608 1.5>]}
        GreenCurve {keys [<at 0.0057 2>,<at 0.7608 1.5>]}
        RedCurve {keys [<at 0.0057 0.5>,<at 0.7608 0.375>]}
       }}
      CurveIndex=$Particles.NormalizedRibbonPos
     )
     Scale Factor=<1 1 1 1>
    )) {module=Color001}
   /SetVariables(
    Particles.Position=&AddVectorToPosition(Position=$Particles.GoalPos, Vector=$Particles.SpiralOffset)
    Particles.Position2:NiagaraPosition=&AddVectorToPosition(Position=$Particles.GoalPos2, Vector=$Particles.SpiralOffset2)
    Particles.Position3:NiagaraPosition=&AddVectorToPosition(Position=$Particles.GoalPos3, Vector=$Particles.SpiralOffset3)
   ) {isSetParameters=true, module=SetVariables_#15}
   &InheritSourceMovement(
    Applied Position Falloff Distance=275
    Applied Position Scale=&VectorFromFloat(Value=&Multiply_Float(
      A=-1
      B=&FloatFromCurve(
       FloatCurve=DataInterface() {propertyValues {Curve {keys [
           {arriveTangent=0.9742, interpMode=RCIM_Cubic, leaveTangent=0.9742, tangentMode=RCTM_User, time=0, value=0}
           <at 0.5 1 RCIM_Cubic>
           {arriveTangent=-0.7756, interpMode=RCIM_Cubic, leaveTangent=-0.7756, tangentMode=RCTM_User, time=1, value=0}
          ]}}}
       CurveIndex=$Particles.NormalizedRibbonPos
      )
     ))
    ApplyForce=false
    ApplyPosition=true
    ApplyPositionFalloff=true
    ApplyPositionFalloffCurve=false
    Source Velocity:Vector3f=&TransformVector(Vector=$Engine.Owner.Velocity, Source Space=::World, Destination Space=::Local)
   )
   &self:InheritVelocity2(
    Applied Position Falloff Distance=275
    Applied Position Scale:Vector3f=&VectorFromFloat(Value=&Multiply_Float(
      A=-1
      B=&FloatFromCurve(
       FloatCurve=DataInterface() {propertyValues {Curve {keys [
           {interpMode=RCIM_Cubic, leaveTangentWeight=0.0794, tangentMode=RCTM_User, time=0, value=0}
           {arriveTangent=0.875, interpMode=RCIM_Cubic, leaveTangent=0.875, time=0.2, value=0.5}
           <at 0.4 0.35 RCIM_Cubic>
           {arriveTangent=-0.5833, interpMode=RCIM_Cubic, leaveTangent=-0.5833, time=0.7, value=0.5}
           {arriveTangent=-0.2676, arriveTangentWeight=0.7681, interpMode=RCIM_Cubic, leaveTangent=-0.2676, tangentMode=RCTM_User, time=1, value=0}
          ]}}}
       CurveIndex=$Particles.NormalizedRibbonPos
      )
     ))
    ApplyForce=false
    ApplyPosition=true
    ApplyPositionFalloff=true
    ApplyPositionFalloffCurve=false
   ) {module=InheritVelocity2}
   &self:InheritVelocity3(
    Applied Position Falloff Distance=275
    Applied Position Scale:Vector3f=&VectorFromFloat(Value=&Multiply_Float(
      A=-1
      B=&FloatFromCurve(
       FloatCurve=DataInterface() {propertyValues {Curve {keys [
           {arriveTangent=0.9742, interpMode=RCIM_Cubic, leaveTangent=0.9742, tangentMode=RCTM_User, time=0, value=0}
           <at 0.5 1 RCIM_Cubic>
           {arriveTangent=-0.7756, interpMode=RCIM_Cubic, leaveTangent=-0.7756, tangentMode=RCTM_User, time=1, value=0}
          ]}}}
       CurveIndex=$Particles.NormalizedRibbonPos
      )
     ))
    ApplyForce=false
    ApplyPosition=true
    ApplyPositionFalloff=true
    ApplyPositionFalloffCurve=false
   ) {module=InheritVelocity3}
   &SolveForcesAndVelocity(Clamp Velocity=false, Limit Acceleration=false)
   /SetVariables(Particles.RibbonFacing=&NormalizeVector(
     Vector To Normalize=&TransformVector(Vector=$Particles.SpiralOffset, Destination Space=::Simulation)
     Safe Normalize=true
     Safe Fallback Threshold=0
     Safe Fallback Vector=<0 0 0.001>
    )) {isSetParameters=true, module=SetVariables_#16}
   &ScaleRibbonWidth(Ribbon Width Scale=&Multiply_Float(
     A=&FloatFromCurve(
      FloatCurve=DataInterface() {propertyValues {Curve {keys [
          {arriveTangent=4.2529, interpMode=RCIM_Cubic, leaveTangent=4.2529, tangentMode=RCTM_User, time=0, value=0.45}
          {interpMode=RCIM_Cubic, tangentMode=RCTM_User, time=0.15, value=2.025}
          {
           arriveTangent=-4.7574
           arriveTangentWeight=0.2872
           interpMode=RCIM_Cubic
           leaveTangent=-4.7574
           leaveTangentWeight=0.1567
           tangentMode=RCTM_User
           time=0.6
           value=1.425
          }
          {arriveTangent=-1.6291, interpMode=RCIM_Cubic, leaveTangent=-1.6291, tangentMode=RCTM_User, time=1, value=0.225}
         ]}}}
      CurveIndex=$Particles.NormalizedRibbonPos
     )
     B=&Waveform(
      Global Phase=&Add_Float(A=$Emitter.Age, B=$Particles.NormalizedRibbonPos)
      "Waveform [1]"=::NewEnumerator9
      "[1] Phase Scale"=5
      "[1] Frequency"=0.25
      "[1] Amplitude Min/Max"=<0.7 1.5>
     )
    ))
   /SetVariables(
    Particles.RibbonWidth2:NiagaraFloat=&Multiply_Float(
     A=&FloatFromCurve(
      FloatCurve=DataInterface() {propertyValues {Curve {keys [
          {arriveTangent=16.2993, interpMode=RCIM_Cubic, leaveTangent=16.2993, tangentMode=RCTM_User, time=0, value=1}
          {interpMode=RCIM_Cubic, tangentMode=RCTM_User, time=0.1, value=1.9}
          {arriveTangent=0.2, interpMode=RCIM_Cubic, leaveTangent=0.2, time=0.3, value=1.55}
          {
           arriveTangent=-6.0526
           arriveTangentWeight=0.2872
           interpMode=RCIM_Cubic
           leaveTangent=-6.0526
           leaveTangentWeight=0.1567
           tangentMode=RCTM_User
           time=0.6
           value=2
          }
          {arriveTangent=-0.6719, interpMode=RCIM_Cubic, leaveTangent=-0.6719, tangentMode=RCTM_User, time=1, value=1}
         ]}}}
      CurveIndex=$Particles.NormalizedRibbonPos
      Scale Curve=11
     )
     B=&Waveform(
      Global Phase=&Add_Float(A=$Emitter.Age, B=$Particles.NormalizedRibbonPos)
      "Waveform [1]"=::NewEnumerator12
      "[1] Phase Scale"=1.5
      "[1] Frequency"=1
      "[1] Amplitude Min/Max"=<0.7 1.2>
     )
    )
    Particles.RibbonWidth3:NiagaraFloat=&Multiply_Float(
     A=&FloatFromCurve(
      FloatCurve=DataInterface() {propertyValues {Curve {keys [
          {arriveTangent=32.4097, interpMode=RCIM_Cubic, leaveTangent=32.4097, tangentMode=RCTM_User, time=0, value=0.45}
          {interpMode=RCIM_Cubic, tangentMode=RCTM_User, time=0.15, value=2.025}
          {arriveTangent=-0.0556, interpMode=RCIM_Cubic, leaveTangent=-0.0556, time=0.4, value=0.55}
          {arriveTangentWeight=0.2872, interpMode=RCIM_Cubic, leaveTangentWeight=0.1567, tangentMode=RCTM_User, time=0.6, value=2}
          {arriveTangent=-1.6291, interpMode=RCIM_Cubic, leaveTangent=-1.6291, tangentMode=RCTM_User, time=1, value=1}
         ]}}}
      CurveIndex=$Particles.NormalizedRibbonPos
      Scale Curve=5
     )
     B=&Waveform(
      Global Phase=&Add_Float(A=$Emitter.Age, B=$Particles.NormalizedRibbonPos)
      "Waveform [1]"=::NewEnumerator9
      "[1] Phase Scale"=5
      "[1] Frequency"=0.25
      "[1] Amplitude Min/Max"=<0.3 1>
     )
    )
   ) {isSetParameters=true}
   /SetVariables(Particles.GlowWidth:NiagaraFloat=&Multiply_Float(A=$Particles.RibbonWidth, B=1.5)) {isSetParameters=true, module=SetVariables_#17}
   /SetVariables(Particles.GlowColor=&MultiplyLinearColor(
     Linear Color=&MultiplyLinearColor(Linear Color=$Particles.Color, Scale Factor=&MakeVector4_FromLinearColor(LinearColor=<rgba 1 1 1 1>))
     Scale Factor=&MakeVector4(
      X=1
      Y=1
      Z=1
      W=&FloatFromCurve(
       FloatCurve=DataInterface() {propertyValues {Curve {keys [
           {arriveTangent=8.2127, interpMode=RCIM_Cubic, leaveTangent=8.2127, tangentMode=RCTM_User, time=0, value=0}
           {arriveTangent=0.4596, arriveTangentWeight=0.232, leaveTangentWeight=2.2538, time=0.1, value=1}
           {arriveTangent=0.0486, interpMode=RCIM_Cubic, leaveTangent=0.0486, tangentMode=RCTM_User, time=0.75, value=1}
           {arriveTangent=-4.4982, interpMode=RCIM_Cubic, leaveTangent=-4.4982, tangentMode=RCTM_User, time=1, value=0}
          ]}}}
       CurveIndex=$Particles.NormalizedRibbonPos
       Scale Curve=&OneMinusFloat(Float=$System.Timeline.NormalizedTime)
      )
     )
    )) {isSetParameters=true, module=SetVariables_#18}
   &DynamicMaterialParameters(
    Index 0 Param 1=1
    Index 1 Param 2=1
    Index 1 Param 3=&FloatFromCurve(
     FloatCurve=DataInterface() {propertyValues {Curve {keys [
         {arriveTangent=-1.0453, interpMode=RCIM_Cubic, leaveTangent=-1.0453, tangentMode=RCTM_User, time=0, value=0.5}
         {arriveTangentWeight=1.8964, interpMode=RCIM_Cubic, tangentMode=RCTM_User, time=0.8, value=0}
        ]}}}
     CurveIndex=$Particles.PartitionParticles.Normalized Index In Partition
    )
    Module.Index0Float0Write=true
    Module.Index0Float1Write=true
    Module.Index0Float2Write=true
    Module.Index0Float3Write=true
    Module.Index1Float0Write=true
    Module.Index1Float1Write=true
    Module.Index1Float2Write=true
    Module.Index1Float3Write=true
    Write Parameter Index 1=::NewEnumerator1
   )
   &ScaleColor(
    Curve Index=$Particles.NormalizedRibbonPos
    Linear Color Curve=DataInterface() {propertyValues {
      AlphaCurve {keys [{arriveTangent=1.6785, interpMode=RCIM_Cubic, time=0, value=1}, <at 0.6031 1>, <at 0.7765 0.3>, <at 1 0>]}
      BlueCurve {keys [<at 0.0057 1>,<at 0.7608 1.5>]}
      GreenCurve {keys [<at 0.0057 1>,<at 0.7608 1.5>]}
      RedCurve {keys [<at 0.0057 0.5029>,<at 0.7608 0.375>]}
     }}
    Scale Mode=::NewEnumerator2
   ) {module=ScaleColor001}
   &ScaleColor(
    Scale RGB=&MakeVectorFromLinearColorRGB(Color=&MultiplyLinearColor(
      Linear Color=$StackContext.Color
      Scale Factor=&MakeVector4_FromFloat(Float=&FNTimeOfDayScalar(Night Scale=1, Morning Scale=1, Evening Scale=1.05))
     ))
    ScaleA=true
    ScaleRGB=true
   ) {module=ScaleColor002}
   /SetVariables(
    Particles.Color2=/Niagara/DynamicInputs/LinearColor/MakeLinearColorFromVector4(Vector:Vector4f=&Multiply_Vector4(
      A=&MakeVector4_FromLinearColor(LinearColor=&ScaleLinearColorByCurve(
        InitialValue=$Particles.Initial.Color2
        ScaleCurve=DataInterface() {propertyValues {
          AlphaCurve {keys [<at 0.0365 1 RCIM_Cubic>,<at 0.7167 1>,<at 1 0>]}
          BlueCurve {keys [<at 0.0057 1>,<at 0.7608 1.5>]}
          GreenCurve {keys [<at 0.0057 2>,<at 0.7608 1.5>]}
          RedCurve {keys [<at 0.0057 0.5>,<at 0.7608 0.375>]}
         }}
        CurveIndex=$Particles.NormalizedRibbonPos
       ))
      B=&MakeVector4FromVectorAndFloat(
       "Vector (RGB)"=&MakeVectorFromLinearColorRGB(Color=&MultiplyLinearColor(
         Linear Color=&ColorFromCurve(
          ColorCurve=DataInterface() {propertyValues {
            AlphaCurve {keys [<at 0.0365 1 RCIM_Cubic>,<at 0.7167 1>,<at 1 0>]}
            BlueCurve {keys [<at 0.0057 1>,<at 0.7608 1.5>]}
            GreenCurve {keys [<at 0.0057 2>,<at 0.7608 1.5>]}
            RedCurve {keys [<at 0.0057 0.5>,<at 0.7608 0.375>]}
           }}
          CurveIndex=$Particles.NormalizedRibbonPos
         )
         Scale Factor=&MakeVector4_FromFloat(Float=&FNTimeOfDayScalar(Night Scale=1, Morning Scale=1, Evening Scale=1.05))
        ))
       "Float (Alpha)"=&MakeCustomFloatFromBool(
        Bool=&SetBoolByFloatComparison(A=$User.IsPickaxe, B=0)
        True Float=&FloatFromCurve(
         FloatCurve=DataInterface() {propertyValues {Curve {
            defaultValue=1
            keys [
             {interpMode=RCIM_Cubic, tangentMode=RCTM_User, time=0, value=0}
             {arriveTangent=0.2356, interpMode=RCIM_Cubic, leaveTangent=0.2356, tangentMode=RCTM_User, time=0.5, value=1}
            ]
           }}}
         CurveIndex=$Particles.NormalizedRibbonPos
        )
        False Float=1
       )
      )
     ))
    Particles.Color3:LinearColor=&MakeLinearColorFromVector4(Vector:Vector4f=&Multiply_Vector4(
      A=&Multiply_Vector4(
       A=&MakeVector4_FromLinearColor(LinearColor=&ColorFromCurve(
         ColorCurve=DataInterface() {propertyValues {
           AlphaCurve {keys [<at 0.0365 1 RCIM_Cubic>,<at 0.7167 1>,<at 1 0>]}
           BlueCurve {keys [<at 0.0057 1>,<at 0.7608 1.5>]}
           GreenCurve {keys [<at 0.0057 2>,<at 0.7608 1.5>]}
           RedCurve {keys [<at 0.0057 0.5>,<at 0.7608 0.375>]}
          }}
         CurveIndex=$Particles.NormalizedRibbonPos
        ))
       B=&MakeVector4_FromLinearColor(LinearColor=&ColorFromCurve(
         ColorCurve=DataInterface() {propertyValues {
           AlphaCurve {keys [
             <at 0.0059 0>
             <at 0.1783 0.3>
             {arriveTangent=1.6478, interpMode=RCIM_Cubic, leaveTangent=1.6478, time=0.2798, value=1}
             <at 0.6031 1>
             <at 0.7069 0.3>
             <at 1 0>
            ]}
           BlueCurve {keys [<at 0 1.3564>,<at 0.2654 1.5>,<at 1 1>]}
           GreenCurve {keys [<at 0 1.5>,<at 0.2654 1.5>,<at 1 0.756>]}
           RedCurve {keys [<at 0 0.2758>,<at 0.2654 0.375>,<at 1 0.2942>]}
          }}
         CurveIndex=$Particles.Initial.NormalizedRibbonPos
        ))
      )
      B=&MakeVector4FromVectorAndFloat(
       "Vector (RGB)"=&VectorFromFloat(Value=&FNTimeOfDayScalar(Night Scale=1, Morning Scale=1, Evening Scale=1.05))
       "Float (Alpha)"=1
      )
     ))
   ) {isSetParameters=true}
   /SetVariables(Particles.Dynamic3Index1:Vector4f=&MakeVector4(
     X=0
     Y=1
     Z=&Multiply_Float(
      A=&FloatFromCurve(
       FloatCurve=DataInterface() {propertyValues {Curve {
          defaultValue=1
          keys [
           {arriveTangent=-5.4241, interpMode=RCIM_Cubic, leaveTangent=-5.4241, tangentMode=RCTM_User, time=-0.0188, value=0.9}
           {arriveTangent=0.0964, interpMode=RCIM_Cubic, leaveTangent=0.0964, time=0.21, value=0.3}
           {arriveTangent=6.5567, interpMode=RCIM_Cubic, leaveTangent=-6.8813, tangentMode=RCTM_Break, time=0.5, value=0.95}
           <at 0.7882 0.35 RCIM_Cubic>
           {arriveTangent=8.0581, tangentMode=RCTM_User, time=1, value=0.95}
          ]
         }}}
       CurveIndex=$Particles.NormalizedRibbonPos
       Scale Curve=&Waveform("Waveform [1]"=::NewEnumerator0, "[1] Frequency"=0.3, "[1] Amplitude Min/Max"=<1 1.5>)
      )
      B=&Waveform(
       Global Phase=&Add_Float(A=$Emitter.Age, B=$Particles.NormalizedRibbonPos)
       "Waveform [1]"=::NewEnumerator12
       "[1] Phase Scale"=3
       "[1] Frequency"=0.25
       "[1] Amplitude Min/Max"=<1 1.75>
      )
     )
     W=0
    )) {isSetParameters=true}
  ]
  rendererClasses [&NiagaraRibbonRendererProperties]
  renderers [
   {rendererClass=&NiagaraRibbonRendererProperties}
   {rendererClass=&NiagaraRibbonRendererProperties, rendererIndex=1}
   {rendererClass=&NiagaraRibbonRendererProperties, rendererIndex=2}
   {rendererClass=&NiagaraRibbonRendererProperties, rendererIndex=3}
  ]
 }
 Pickaxe_Glows {
  emitterUpdateScript [
   &EmitterState(
    Inactive Response=::NewEnumerator1
    Life Cycle Mode=::NewEnumerator1
    Loop Behavior=::NewEnumerator1
    Loop Duration Mode=::NewEnumerator1
    UseLoopDelay=false
   )
   &SpawnBurst_Instantaneous(Spawn Count=16, Use Spawn Probability=false, UseLoopCountLimit=true)
  ]
  particleSpawnScript [
   &InitializeParticle(
    Color Mode=::NewEnumerator0
    Lifetime=0.25
    Mesh Scale Mode=::NewEnumerator3
    Sprite Size Mode=::NewEnumerator3
    Uniform Sprite Size=5
    UsePositionOffset=false
    Write Material Random=true
   )
   &PartitionParticles(Number Of Partitions=16)
   /SetVariables(Particles.VisibilityTag=&MakeCustomIntFromBool(
     Bool=&Boolean_OrOperation(Bool1=&Boolean_OrOperation(Bool1=$User.bIsFrontEnd, Bool2=$User.IsFrontEnd), Bool2=$User.bIsFrontEndPreview)
     True INT=1
     False INT=0
    )) {isSetParameters=true}
  ]
  particleUpdateScript [
   &ParticleState(Kill Particles When Lifetime Has Elapsed=false)
   &ScaleMeshSize(Scale Factor=&VectorFromFloat(Value=$System.MeshScaleRemap))
   &SkeletalMeshLocation(Bone / Socket Index=$Engine.ExecIndex, Bone Sampling Mode=::NewEnumerator4, MaskSpawnGroup=false, Skeletal Mesh=$System.Skeletal Mesh2)
   /SetVariables(Particles.MeshIndex=&ReturnExecIndex()) {isSetParameters=true}
  ]
  rendererClasses [&NiagaraMeshRendererProperties]
  renderers [
   {rendererClass=&NiagaraMeshRendererProperties, rendererIndex=0}
   {rendererClass=&NiagaraMeshRendererProperties, rendererIndex=1}
  ]
 }
}
summary {
 systemName=NS_Pickaxe_MicStand_IvyStub_Idle_FX
 emitters [
  {emitterName=Mist_Sprites_local, rendererClasses [&NiagaraSpriteRendererProperties]}
  {emitterName=NE_MicRibbon, rendererClasses [&NiagaraRibbonRendererProperties]}
  {emitterName=NE_SpiralRibbon001, rendererClasses [&NiagaraRibbonRendererProperties]}
  {emitterName=NE_Motes, rendererClasses [&NiagaraSpriteRendererProperties]}
  {bEnabled=false, emitterName=Mist_Sprites001, rendererClasses [&NiagaraSpriteRendererProperties]}
  {emitterName=Pickaxe_Glows, rendererClasses [&NiagaraMeshRendererProperties]}
 ]
 userVariables [
  User.bIsFrontEnd {type=NiagaraBool}
  User.bIsFrontEndPreview {type=NiagaraBool}
  User.IsFrontEnd {type=NiagaraBool}
  User.IsPickaxe {type=NiagaraFloat, defaultValue=0, description=""}
 ]
}
